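; RiskyReelsPoster is a building deco actor that shows a poster only while a named calendar event is active.
; On BeginPlay the server sets PosterState to hidden, then watches the calendar event, switching the state to visible when it starts and hidden when it ends.
; The PosterState OnRep function hides or shows the static mesh on replication.
; Two debug custom events force either state.
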(event EventBeginPlay
  ;[1 0 0] beginplay
  (Networking|SwitchHasAuthority
    (:Authority
      (Variables|Default|SetPosterState 0)
      ;[1 1 1] when event active set poster to be visible
      (MCPContext|WatchCalendarEvent (Variables|Default|GetCalendarEventName)
        (:OnEventActive
          (Variables|Default|SetPosterState 1))
        (:OnEventEnded
          (Variables|Default|SetPosterState 0))))))

(event Custom|posterdebug_0
  (Variables|Default|SetPosterState 0))

(event Custom|posterdebug_1
  (Variables|Default|SetPosterState 1))

(fn OnRep_PosterState
  (bind _staticmeshcomponent (Variables|Debug|GetStaticMeshComponent))
  (switch int (Variables|Default|GetPosterState)
    (:0
      (Development|SetHiddeninGame _staticmeshcomponent true false))
    (:1
      (Development|SetHiddeninGame _staticmeshcomponent false false))))

(fn ConstructionScript)
